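; Editor-only foliage cluster Blueprint for the RustBucket open desert environment, circular variant E.
; It holds five instanced static mesh components and returns them through GetSimpleDestructiblePropComponents, alongside a collision mask of zero.
; On BeginPlay it calls the parent, then destroys itself so it does not exist on the PIE server.
; Construction defers to the parent class; overlap and tick events are empty.
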(event EventBeginPlay
  (|CallParentFunction)
  (Actor|DestroyActor) ; This actor is intended to only be used in the editor.\nWe destroy in BeginPlay to make sure we don't exist on the PIE server.
)

(event Collision|EventActorBeginOverlap
  (params (OtherActor :type Actor%20Object%20Reference)))

(event EventTick
  (params (DeltaSeconds :type Float%20%28single-precision%29)))

(fn GetSimpleDestructiblePropCollisionMask
  (return 0))

(fn GetSimpleDestructiblePropComponents
  (return (Utilities|Array|MakeArray (Variables|Default|GetInstancedStaticMesh1) (Variables|Default|GetInstancedStaticMesh2) (Variables|Default|GetInstancedStaticMesh3) (Variables|Default|GetInstancedStaticMesh4) (Variables|Default|GetInstancedStaticMesh0))))

(fn ConstructionScript
  (|CallParentFunction))
